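# UI material for an on/off toggle switch in the matchmaking settings, driven by the IsOn progress parameter.
# A rounded mask scaled by vertex alpha shapes the track, with an inset inner mask for the background.
# An aspect-corrected SDF circle forms the sliding knob, with an eased outline thickness and a second, expanding fill circle.
# These masks blend colors and opacities from the shared overlay backplates function into an alpha-composite output.
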Material /MatchmakingSettingsUI/MatchmakingSettings/SettingEntries/Toggle/Materials/M_UIKit_Toggle
graph {
 properties {bAutomaticallySetUsageInEditor=false, blendMode=BLEND_AlphaComposite, materialDomain=MD_UI, usedShadingModels=Default Lit}
 nodes [
  n1 = Multiply(
   A=MaterialFunctionCall(Corner Radius=Constant() {r=0.5}, Radius in Pixels=Constant()) {materialFunction=&MF_UI_CoreUI_RoundedMask}.Shape Mask
   B=VertexColor().A
  )
  n2 = MaterialFunctionCall() {materialFunction=/MatchmakingSettingsUI/MatchmakingSettings/Materials/MF_UI_MM_OverlayBackplates}
  n3 = NamedRerouteUsage()
  n4 = OneMinus(None=Divide(A=Constant() {r=0.05}, B=Max(A=@n3) {constB=0}))
  n5 = Constant() {r=0.5}
  n6 = Constant()
  n7 = Multiply(
   A=NamedRerouteDeclaration(None=MaterialFunctionCall(Size=@n4, Corner Radius=@n5, Radius in Pixels=@n6) {materialFunction=&MF_UI_CoreUI_RoundedMask}.Shape Mask) {name=Inner Background Mask, nodeColor=<rgba 0 0 0 1>}
   B=NamedRerouteUsage()
  )
  n8 = LinearInterpolate(A=@n2.C Opacity, B=@n2.C Opacity, Alpha=NamedRerouteUsage())
  n9 = NamedRerouteUsage()
  Multiply(
   A=@n1
   B=LinearInterpolate(
    A=LinearInterpolate(A=@n2.B Weak Opacity, B=@n2.B Weak Opacity, Alpha=@n7)
    B=LinearInterpolate(A=@n8, Alpha=@n9) {constA=0.5}
    Alpha=NamedRerouteUsage()
   )
  )
  n10 = LinearInterpolate(A=@n2.C Color, B=@n2.C Color, Alpha=NamedRerouteUsage())
  n11 = Constant3Vector() {constant=<rgba 0.0048 0.004 0.0116 1>}
  n12 = NamedRerouteUsage()
  Multiply(
   A=LinearInterpolate(
    A=LinearInterpolate(A=@n2.B Weak Color, B=@n2.B Weak Color, Alpha=@n7)
    B=LinearInterpolate(A=@n10, B=@n11, Alpha=@n12) {constA=0.5}
    Alpha=NamedRerouteUsage()
   )
   B=@n1
  )
  NamedRerouteDeclaration(None=ScalarParameter() {defaultValue=1, desc=Progress, parameterName=IsOn, sliderMax=1}) {name=Progress Ratio, nodeColor=<rgba 0 0 0 1>}
  n13 = NamedRerouteDeclaration(None=MaterialFunctionCall!(Dimensions) {materialFunction=/FortUILibrary/MaterialFunctions/Basics/MF_UI_GetAspectRatio}.Min Ratio) {name=Aspect Ratio, nodeColor=<rgba 0 0 0 1>}
  n14 = Multiply(A=Constant() {r=0.5}, B=OneMinus(None=ComponentMask(None=@n13) {g=true}))
  n15 = LinearInterpolate(A=@n14, B=Multiply(A=@n14) {constB=-1}, Alpha=NamedRerouteUsage()) {desc=Offset}
  n16 = Constant()
  n17 = MaterialFunctionCall(UV=Add(A=MaterialFunctionCall() {materialFunction=&GetUserInterfaceUV}.Normalized UV, B=AppendVector(A=@n15, B=@n16))) {materialFunction=&MF_UI_UVScaling}
  n18 = Constant() {r=0.625}
  n19 = NamedRerouteUsage()
  n20 = MaterialFunctionCall(
   UVs=@n17.Scaled to Fit
   Size=@n18
   OutlineThickness=LinearInterpolate(Alpha=MaterialFunctionCall("Value (0-1)"=@n19) {materialFunction=&MF_UI_EaseCurves}.EaseOutCubic) {constA=0.1, constB=0}
   OutlineAlignment=Constant() {r=-1}
  ) {materialFunction=&MF_UI_Circle}
  NamedRerouteDeclaration(None=MaterialFunctionCall(DistanceField=@n20.SDFOutline) {materialFunction=&MF_UI_AntiAliasing}.Inner) {name=Toggle Mask, nodeColor=<rgba 0 0 0 1>}
  NamedRerouteDeclaration(None=@n20.Fill) {name=Toggle Fill Mask, nodeColor=<rgba 0 0 0 1>}
  n21 = NamedRerouteUsage()
  NamedRerouteDeclaration(None=MaterialFunctionCall(UVs=@n17.Scaled to Fit, Size=LinearInterpolate(A=@n18, Alpha=@n21) {constB=3}) {materialFunction=&MF_UI_Circle}.Fill) {name=Toggle Expand Fill, nodeColor=<rgba 0 0 0 1>}
 ]
}
